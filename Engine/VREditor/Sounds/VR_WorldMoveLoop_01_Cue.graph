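# VR_WorldMoveLoop_01_Cue is a SoundCue in the VR Editor sounds, named as the loop for world movement.
# A single Wave Player root plays the VR_WorldMoveLoop_01 wave, with no mixing, modulation or randomization nodes.
# The player does not enable looping itself, so any looping depends on the wave asset.
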SoundCue /Engine/VREditor/Sounds/VR_WorldMoveLoop_01_Cue
root=WavePlayer_21
nodes [{id=WavePlayer_21, properties {soundWaveAssetPtr=&/VR_WorldMoveLoop_01}}]
properties {maxDistance=2097152}
