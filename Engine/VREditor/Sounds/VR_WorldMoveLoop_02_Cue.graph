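# Sound cue for the VR Editor world-movement loop sound, going by its name and location.
# A single wave player node feeds the root directly, playing the VR_WorldMoveLoop_02 sound wave with no modulation, mixing or attenuation nodes.
# The wave player does not enable looping itself, and the cue records a very large max distance.
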SoundCue /Engine/VREditor/Sounds/VR_WorldMoveLoop_02_Cue
root=WavePlayer_22
nodes [{id=WavePlayer_22, properties {soundWaveAssetPtr=&/VR_WorldMoveLoop_02}}]
properties {maxDistance=2097152}
