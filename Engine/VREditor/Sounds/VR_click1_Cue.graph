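# Sound cue for a click sound used by the VR Editor.
# A single wave player feeds the cue output directly, playing the VR_click1 sound wave with no modulation, mixing or randomization.
# The cue sets a very large max distance.
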SoundCue /Engine/VREditor/Sounds/VR_click1_Cue
root=WavePlayer_0
nodes [{id=WavePlayer_0, properties {soundWaveAssetPtr=&/VR_click1}}]
properties {maxDistance=2097152}
