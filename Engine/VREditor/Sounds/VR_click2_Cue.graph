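# VR_click2_Cue is a Sound Cue for a click sound in the VR Editor content.
# A single Wave Player node plays the VR_click2 sound wave directly as the cue root, with no modulation, mixing or randomization.
# The cue carries a very large max distance.
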SoundCue /Engine/VREditor/Sounds/VR_click2_Cue
root=WavePlayer_2
nodes [{id=WavePlayer_2, properties {soundWaveAssetPtr=&/VR_click2}}]
properties {maxDistance=2097152}
