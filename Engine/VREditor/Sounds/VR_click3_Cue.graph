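# Sound cue for a click sound in the VR Editor, wrapping a single sound wave.
# A lone Wave Player node plays the VR_click3 wave as the root, with no modulation, mixing or randomization.
# The cue's max distance is set to a very large value.
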SoundCue /Engine/VREditor/Sounds/VR_click3_Cue
root=WavePlayer_1
nodes [{id=WavePlayer_1, properties {soundWaveAssetPtr=&/VR_click3}}]
properties {maxDistance=2097152}
