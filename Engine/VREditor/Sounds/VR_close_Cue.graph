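# VR Editor sound cue that plays the VR close sound, likely for closing interface elements in VR editing.
# A single Wave Player node feeds the VR_close sound wave straight to the output, with no mixing, randomization or modulation.
# The cue sets a very large maximum distance.
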SoundCue /Engine/VREditor/Sounds/VR_close_Cue
root=WavePlayer_3
nodes [{id=WavePlayer_3, properties {soundWaveAssetPtr=&/VR_close}}]
properties {maxDistance=2097152}
