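# VR_confirm_Cue is a sound cue in the VR Editor content that plays the VR_confirm sound.
# A single wave player node serves as the root and feeds the output directly, with no modulation, randomization or mixing.
# The cue carries a very large max distance value.
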SoundCue /Engine/VREditor/Sounds/VR_confirm_Cue
root=WavePlayer_4
nodes [{id=WavePlayer_4, properties {soundWaveAssetPtr=&/VR_confirm}}]
properties {maxDistance=2097152}
