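# Sound cue VR_grab_Cue in the engine's VR Editor sounds, named for the grab sound.
# It consists of a single wave player node as root, playing the VR_grab sound wave with no mixing, modulation or randomization.
# The cue sets a very large max distance.
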SoundCue /Engine/VREditor/Sounds/VR_grab_Cue
root=WavePlayer_5
nodes [{id=WavePlayer_5, properties {soundWaveAssetPtr=&/VR_grab}}]
properties {maxDistance=2097152}
