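# Sound cue for the VR Editor's negative feedback sound, found in the engine's VREditor sounds.
# A single wave player node serves as the root and plays the VR_negative wave directly, with no mixing, randomization or modulation.
# The cue sets a very large max distance.
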SoundCue /Engine/VREditor/Sounds/VR_negative_Cue
root=WavePlayer_6
nodes [{id=WavePlayer_6, properties {soundWaveAssetPtr=&/VR_negative}}]
properties {maxDistance=2097152}
