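# Sound cue for the VR Editor's object rotate sound, judging by its name a loop played while rotating objects.
# A single wave player feeds the VR_objectRotateLoop_01 sound wave straight to the output, with no mixing, modulation or looping node.
# The cue sets a very large max distance.
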SoundCue /Engine/VREditor/Sounds/VR_objectRotateLoop_01_Cue
root=WavePlayer_8
nodes [{id=WavePlayer_8, properties {soundWaveAssetPtr=&/VR_objectRotateLoop_01}}]
properties {maxDistance=2097152}
