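# Sound cue for the VR Editor, named as the looping sound for rotating an object.
# A single wave player feeds the output directly, playing the VR_objectRotateLoop_02 sound wave with no mixing, modulation or attenuation nodes.
# The cue carries a very large max distance, and the wave player does not set its own looping flag.
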SoundCue /Engine/VREditor/Sounds/VR_objectRotateLoop_02_Cue
root=WavePlayer_9
nodes [{id=WavePlayer_9, properties {soundWaveAssetPtr=&/VR_objectRotateLoop_02}}]
properties {maxDistance=2097152}
