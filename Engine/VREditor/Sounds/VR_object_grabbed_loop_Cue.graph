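# Sound cue in the engine's VR Editor content, named as the loop sound for a grabbed object.
# A single Wave Player node is the root and plays the VR_object_grabbed_loop sound wave, with no looping, mixing or modulation nodes in the cue.
# No attenuation is set here, and the max distance is a very large value.
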SoundCue /Engine/VREditor/Sounds/VR_object_grabbed_loop_Cue
root=WavePlayer_7
nodes [{id=WavePlayer_7, properties {soundWaveAssetPtr=&/VR_object_grabbed_loop}}]
properties {maxDistance=2097152}
