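# Sound cue for the VR Editor's open sound, stored in the engine's VREditor content.
# A single wave player feeds the cue output directly with the VR_open sound wave, with no randomization, mixing or modulation.
# The cue carries a very large max distance value.
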SoundCue /Engine/VREditor/Sounds/VR_open_Cue
root=WavePlayer_10
nodes [{id=WavePlayer_10, properties {soundWaveAssetPtr=&/VR_open}}]
properties {maxDistance=2097152}
