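# Sound cue for the VR Editor that plays a descending scale sound, apparently a Shepard tone judging by its name.
# A single Wave Player root node plays the VR_shep_scale_down_01 sound wave directly, with no randomization, modulation or mixing nodes.
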SoundCue /Engine/VREditor/Sounds/VR_shep_scale_down_01_Cue
root=WavePlayer_11
nodes [{id=WavePlayer_11, properties {soundWaveAssetPtr=&/VR_shep_scale_down_01}}]
properties {maxDistance=2097152}
