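# Sound cue in the Engine's VR Editor sounds that plays the VR_shep_scale_down_02 wave, a scale-down sound judging by its name.
# A single Wave Player node serves as the root, with no randomization, mixing or modulation.
# The cue sets a very large max distance.
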SoundCue /Engine/VREditor/Sounds/VR_shep_scale_down_02_Cue
root=WavePlayer_12
nodes [{id=WavePlayer_12, properties {soundWaveAssetPtr=&/VR_shep_scale_down_02}}]
properties {maxDistance=2097152}
