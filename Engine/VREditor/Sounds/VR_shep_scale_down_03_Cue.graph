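# Sound cue for the VR Editor that plays the descending scale sample VR_shep_scale_down_03, likely a Shepard-tone step judging by its name.
# A single wave player node is the root, with no mixing, modulation or randomization.
# The cue sets a very large max distance.
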SoundCue /Engine/VREditor/Sounds/VR_shep_scale_down_03_Cue
root=WavePlayer_13
nodes [{id=WavePlayer_13, properties {soundWaveAssetPtr=&/VR_shep_scale_down_03}}]
properties {maxDistance=2097152}
