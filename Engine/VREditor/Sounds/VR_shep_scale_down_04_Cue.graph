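# Sound cue for the VR Editor that plays a descending scale sound, the fourth variant by its name, apparently a Shepard-tone step.
# A single wave player node feeds the root directly, playing the VR_shep_scale_down_04 wave with no modulation, mixing or randomization.
# The cue's max distance is set to a very large value.
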SoundCue /Engine/VREditor/Sounds/VR_shep_scale_down_04_Cue
root=WavePlayer_14
nodes [{id=WavePlayer_14, properties {soundWaveAssetPtr=&/VR_shep_scale_down_04}}]
properties {maxDistance=2097152}
