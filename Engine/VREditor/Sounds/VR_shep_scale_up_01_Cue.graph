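# Sound cue for a VR Editor scale-up sound, named as a Shepard-style rising tone.
# A single wave player plays the VR_shep_scale_up_01 sound wave directly as the cue's root, with no mixing, modulation or randomisation.
# The cue overrides its max distance to a very large value.
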SoundCue /Engine/VREditor/Sounds/VR_shep_scale_up_01_Cue
root=WavePlayer_15
nodes [{id=WavePlayer_15, properties {soundWaveAssetPtr=&/VR_shep_scale_up_01}}]
properties {maxDistance=2097152}
