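# Sound cue for the VR editor that plays the VR_shep_scale_up_02 sound, named as a scale-up effect.
# It has a single wave player node as its root, with no mixing, modulation or randomization.
# A very large max distance is set on the cue.
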SoundCue /Engine/VREditor/Sounds/VR_shep_scale_up_02_Cue
root=WavePlayer_16
nodes [{id=WavePlayer_16, properties {soundWaveAssetPtr=&/VR_shep_scale_up_02}}]
properties {maxDistance=2097152}
