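# Sound cue for the VR Editor's scale-up feedback, judging by its name, in a numbered "shep" series.
# A single wave player node feeds the root and plays the VR_shep_scale_up_03 sound wave, with no modulation, mixing or randomization.
# The cue's maximum distance is set extremely large.
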SoundCue /Engine/VREditor/Sounds/VR_shep_scale_up_03_Cue
root=WavePlayer_17
nodes [{id=WavePlayer_17, properties {soundWaveAssetPtr=&/VR_shep_scale_up_03}}]
properties {maxDistance=2097152}
