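# Sound cue in the VR Editor that plays the VR_shep_scale_up_04 wave, apparently a scale-up audio effect.
# A single Wave Player node serves as the root, with no mixing, modulation or randomization.
# The cue carries a very large max distance.
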SoundCue /Engine/VREditor/Sounds/VR_shep_scale_up_04_Cue
root=WavePlayer_18
nodes [{id=WavePlayer_18, properties {soundWaveAssetPtr=&/VR_shep_scale_up_04}}]
properties {maxDistance=2097152}
