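# SoundCue for the VR Editor's teleport sound effect.
# A single wave player node serves as the root and plays the VR_teleport sound wave directly, with no mixing, modulation or randomization.
# The cue sets a very large max distance.
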SoundCue /Engine/VREditor/Sounds/VR_teleport_Cue
root=WavePlayer_19
nodes [{id=WavePlayer_19, properties {soundWaveAssetPtr=&/VR_teleport}}]
properties {maxDistance=2097152}
