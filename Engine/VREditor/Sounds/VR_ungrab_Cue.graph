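# VR_ungrab_Cue is a Sound Cue in the engine's VR Editor sounds, named as the audio for releasing a grab.
# A single Wave Player is the root node and plays the VR_ungrab sound wave, with no randomization, modulation or mixing.
# The cue carries a very large max distance value.
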SoundCue /Engine/VREditor/Sounds/VR_ungrab_Cue
root=WavePlayer_20
nodes [{id=WavePlayer_20, properties {soundWaveAssetPtr=&/VR_ungrab}}]
properties {maxDistance=2097152}
